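# Niagara system for the SandyInland companion's fire emote, which appears to be a short burst of fire breath from the mouth.
# Ember sprites, ribbon flames and SubUV fire sprites launch along sampled skeletal bone rotations, joined by a mouth light and spinning mesh flames.
# Spawn groups select per-layer bones, sizes, colors and forces, while emitter-age curves shape the burst.
# Time-of-day and front-end preview inputs adjust color, brightness and opacity.
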NiagaraSystem /CosmeticCompanions/FX/Builtins/SandyInland/NS_Emote_SandyInland_Fire
dependencies {
 usedDataInterfaces [
  &NiagaraDataInterfaceCurve
  &NiagaraDataInterfaceArrayFloat
  &NiagaraDataInterfaceVector2DCurve
  &NiagaraDataInterfaceColorCurve
  &NiagaraDataInterfaceSpriteRendererInfo
  &NiagaraDataInterfaceSkeletalMesh
 ]
 usedDynamicInputs [
  &RandomRangeVector2D
  &UniformRangedFloat
  &MaskVectorBySpawnGroup
  &MaskVector2dBySpawnGroup
 ]
 usedModules [
  &self:SystemSpawnScript.SystemScriptSource.SystemScriptGraph.NiagaraNodeAssignment_0.SetVariables_#1
  &SampleSkeletalMesh
  &SystemState
 ]
}
emitters {
 Embers001 {
  emitterUpdateScript [
   &EmitterState(Life Cycle Mode=::NewEnumerator1, Loop Behavior=::NewEnumerator1, Loop Duration=0.8, UseLoopDelay=false)
   &SpawnRate(Spawn Probability=0.7, SpawnRate=60, Use Spawn Probability=true)
   &SpawnRate(Spawn Probability=0.7, SpawnGroup=1, SpawnRate=45, Use Spawn Probability=true) {module=SpawnRate001}
  ]
  particleSpawnScript [
   &InitializeParticle(
    Color=<rgba 1 1 1 2>
    Lifetime Max=1.2
    Lifetime Min=0.4
    Lifetime Mode=::NewEnumerator1
    Mass Mode=::NewEnumerator1
    Sprite Size Max=<1.2 2>
    Sprite Size Min=<0.6 1.5>
    Sprite Size Mode=::NewEnumerator2
    UsePositionOffset=false
    Write Material Random=true
   ) {module=InitializeParticle001}
   &ShapeLocation(MaskSpawnGroup=false, Override Seed=false, Sphere Radius=20, UseOwnerScale=true)
   &SkeletalMeshLocation(
    Bone / Socket Index=&SpawnGroup()
    Bone Sampling Mode=::NewEnumerator4
    MaskSpawnGroup=false
    Sampled Position Offset:NiagaraPosition=$Particles.ShapeLocation.ShapePosition
    Skeletal Mesh=$System.Skeletal Mesh Reader
   )
   &AddVelocity(
    Velocity=&RotateVectorByQuaternion(Vector To Rotate=<-0.1 1 0.7>, Quaternion=$Particles.SkeletalMeshLocation.SampledRotation)
    Velocity Speed Scale=&Multiply_Float(
     A=&MergeHelper_FloatSwitch(0=200, 1=200, 2=200, 3=?, Index=&SpawnGroup(), Num=::NewEnumerator1)
     B=&FloatFromCurve(
      FloatCurve=DataInterface() {propertyValues {Curve {
         defaultValue=1
         keys [
          {interpMode=RCIM_Cubic, tangentMode=RCTM_User, time=0.5278, value=1}
          {interpMode=RCIM_Cubic, tangentMode=RCTM_User, time=0.7975, value=0.4513}
         ]
        }}}
      CurveIndex=$Emitter.Age
     )
    )
   )
   &CameraOffset(Camera Offset Amount=&RandomRangeFloat(Minimum=0, Maximum=20))
  ]
  particleUpdateScript [
   &ParticleState()
   &ScaleSpriteSize(
    Uniform Curve Scale=&FloatFromCurve(FloatCurve=DataInterface() {propertyValues {Curve {keys [<at 0 0>,<at 0.17 0>,<at 0.2 1.2>]}}}, CurveIndex=$Emitter.Age)
    Uniform Curve Sprite Scale=DataInterface() {propertyValues {Curve {keys [<at 0 0 RCIM_Cubic>,<at 0.2 1 RCIM_Cubic>,<at 1 0 RCIM_Cubic>]}}}
   ) {module=ScaleSpriteSize001}
   &Drag(
    Drag=&Multiply_Float(
     A=&FloatFromCurve(
      FloatCurve=DataInterface() {propertyValues {Curve {keys [<at 0 0.1666 RCIM_Cubic>, {arriveTangent=2.8788, interpMode=RCIM_Cubic, leaveTangent=2.8788, tangentMode=RCTM_User, time=0.526, value=1}]}}}
      Scale Curve=&MaskFloatBySpawnGroup(Masked Float=5, Passthrough Float=&RandomRangeFloat(Minimum=4, Maximum=8))
     )
     B=1
    )
    Use Linear Drag=true
    Use Rotational Drag=false
   )
   /Niagara/Modules/Update/Forces/CurlNoiseForce(
    Noise Frequency:NiagaraFloat=&RandomRangeFloat(Minimum=35, Maximum=5)
    Noise Strength=&RandomRangeFloat(Minimum=325, Maximum=500)
    Pan Noise=false
    Pan Noise Field=<0.1 0.1 0.5>
   )
   &GravityForce(Coordinate Space=::Simulation, Gravity=&RandomRangeVector(Minimum=<0 0 0>, Maximum=<0 0 500>))
   &SolveForcesAndVelocity(Clamp Velocity=false, Limit Acceleration=false)
   &ScaleSpriteSizeBySpeed(
    Max Scale Factor=<1 2>
    Sample Scale Factor By Curve=true
    Scale Factor Curve=DataInterface() {propertyValues {Curve {keys [
        <at -0.0077 0.0765 RCIM_Cubic>
        {arriveTangent=1.2463, interpMode=RCIM_Cubic, leaveTangent=1.2463, tangentMode=RCTM_User, time=0.9985, value=0.9549}
       ]}}}
    Velocity Threshold=300
   )
   &ScaleColor(
    Scale Alpha=&Multiply_Float(
     A=1
     B=&FloatFromCurve(FloatCurve=DataInterface() {propertyValues {Curve {keys [
          {arriveTangent=0.8687, interpMode=RCIM_Cubic, leaveTangent=0.8687, tangentMode=RCTM_User, time=0.1, value=0}
          {arriveTangent=0.0152, interpMode=RCIM_Cubic, leaveTangent=0.0152, tangentMode=RCTM_User, time=0.4, value=1}
          <at 0.8 0 RCIM_Cubic>
         ]}}})
    )
    Scale RGB=&MakeVectorFromLinearColorRGB(Color=&FNTimeOfDayColor(
      Night Color=<rgba 2.5 0.8193 0.0412 1>
      Morning Color=<rgba 2.5 0.7258 0.048 1>
      Day Color=<rgba 4 0.6667 0 1>
      Evening Color=<rgba 2.5 0.7258 0.048 1>
     ))
    ScaleA=true
    ScaleRGB=true
   ) {module=ScaleColor002}
   /SetVariables(Particles.Glow03_Size=&Multiply_Vector2D_ByFloat(
     Float=&FloatFromCurve(
      FloatCurve=DataInterface() {propertyValues {Curve {keys [
          <at 0 0 RCIM_Cubic>
          <at 0.2 1.3 RCIM_Cubic>
          {arriveTangent=-0.1423, interpMode=RCIM_Cubic, leaveTangent=-0.1423, tangentMode=RCTM_User, time=1, value=0}
         ]}}}
      Scale Curve=10
     )
     Vector2D=<0.75 1>
    )) {isSetParameters=true}
   /SetVariables(Particles.Glow03_Color=&MakeLinearColorFromVectorAndFloat(
     "Vector (RGB)"=&MakeCustomVectorFromBool(
      Bool=$User.bIsFrontEndPreview
      True Vector=&MakeVectorFromLinearColorRGB(Color=<rgba 1 0.28 0.02 1>)
      False Vector=&MakeVectorFromLinearColorRGB(Color=&MultiplyLinearColor(
        Linear Color=&FNTimeOfDayColor(Night Color=<rgba 1 0.3167 0.0354 1>, Morning Color=<rgba 1 0.15 0.02 1>, Day Color=<rgba 1 0.15 0.02 1>, Evening Color=<rgba 1 0.15 0.02 1>)
        Scale Factor=&MakeVector4_FromFloat(Float=&Multiply_Float(A=1, B=&FNTimeOfDayScalar(Night Scale=0.5, Morning Scale=0.5, Evening Scale=0.5)))
       ))
     )
     "Float (Alpha)"=&Multiply_Float(
      A=0.325
      B=&FloatFromCurve(
       FloatCurve=DataInterface() {propertyValues {Curve {keys [<at 0 0>,<at 0.15 0>,<at 0.25 1>]}}}
       CurveIndex=$Emitter.Age
       Scale Curve=&MakeCustomFloatFromBool(Bool=$User.bIsFrontEndPreview, True Float=0.6, False Float=1)
      )
     )
    )) {isSetParameters=true}
  ]
  rendererClasses [&NiagaraSpriteRendererProperties]
  renderers [
   {rendererClass=&NiagaraSpriteRendererProperties}
   {rendererClass=&NiagaraSpriteRendererProperties, rendererIndex=1}
   {rendererClass=&NiagaraSpriteRendererProperties, rendererIndex=2}
   {rendererClass=&NiagaraSpriteRendererProperties, rendererIndex=3}
  ]
 }
 Fire {
  emitterUpdateScript [
   &EmitterState(Life Cycle Mode=::NewEnumerator1, Loop Behavior=::NewEnumerator1, Loop Duration=0.8, UseLoopDelay=false)
   &SpawnRate(SpawnGroup=1, SpawnRate=30, Use Spawn Probability=false) {module=SpawnRate001}
   &SpawnRate(SpawnRate=30, Use Spawn Probability=false)
  ]
  particleSpawnScript [
   /SetVariables(Particles.VisibilityTag=&SpawnGroup()) {isSetParameters=true}
   &InitializeParticle(
    Color=&MaskLinearColor(Mask Source=0, Mask Parameter=?, Mask=0, Masked Linear Color=<rgba 2 0.0667 0 0.5>, Passthrough Linear Color=<rgba 0.0087 0.0075 0.0072 3>)
    Lifetime=&MaskFloat(Mask Source=0, Mask Parameter=?, Mask=0, Masked Float=0.35, Passthrough Float=0.8)
    Ribbon Width=&MaskFloat(
     Mask Source=0
     Mask Parameter=?
     Mask=0
     Masked Float=&FloatFromCurve(
      FloatCurve=DataInterface() {propertyValues {Curve {keys [
          {arriveTangent=6.3399, interpMode=RCIM_Cubic, leaveTangent=6.3399, tangentMode=RCTM_User, time=0.1, value=0}
          {arriveTangent=0.3698, interpMode=RCIM_Cubic, leaveTangent=0.3698, tangentMode=RCTM_User, time=0.2, value=1}
          {arriveTangent=0.4083, interpMode=RCIM_Cubic, leaveTangent=0.043, tangentMode=RCTM_Break, time=0.4636, value=1}
          {arriveTangent=-3.2627, interpMode=RCIM_Cubic, leaveTangent=-3.2627, tangentMode=RCTM_User, time=0.8802, value=0}
         ]}}}
      CurveIndex=$Emitter.Age
      Scale Curve=45
     )
     Passthrough Float=&FloatFromCurve(
      FloatCurve=DataInterface() {propertyValues {Curve {keys [
          {arriveTangent=6.3399, interpMode=RCIM_Cubic, leaveTangent=6.3399, tangentMode=RCTM_User, time=0.1, value=0}
          {arriveTangent=0.3698, interpMode=RCIM_Cubic, leaveTangent=0.3698, tangentMode=RCTM_User, time=0.2, value=1}
          {arriveTangent=0.4083, interpMode=RCIM_Cubic, leaveTangent=0.043, tangentMode=RCTM_Break, time=0.6622, value=1}
          {arriveTangent=-6.1064, interpMode=RCIM_Cubic, leaveTangent=-6.1064, tangentMode=RCTM_User, time=0.8507, value=0}
         ]}}}
      CurveIndex=$Emitter.Age
      Scale Curve=40
     )
    )
    Ribbon Width Mode=::NewEnumerator1
    UsePositionOffset=false
    Write Material Random=true
   ) {module=InitializeParticle001}
   &SkeletalMeshLocation(
    Bone Sampling Mode=::NewEnumerator4
    Interpolate Mesh Sampling=false
    MaskSpawnGroup=false
    Orientation Sampling=::NewEnumerator1
    Sampled Position Offset=&MaskVector(Mask Source=0, Mask Parameter=?, Mask=0, Masked Vector=<0 20 0>, Passthrough Vector=<0 0 0>)
    Skeletal Mesh=$System.Skeletal Mesh Reader
   )
   &DynamicMaterialParameters(
    Index 0 Param 4=0.3
    Module.Index0Param0Write=true
    Module.Index0Param1Write=true
    Module.Index0Param2Write=true
    Module.Index0Param3Write=true
   )
   &AddVelocity(
    Velocity=&RotateVectorByQuaternion(Vector To Rotate=<-0.1 1 0.7>, Quaternion=$Particles.SkeletalMeshLocation.SampledRotation)
    Velocity Speed Scale=200
   )
   /SetVariables(Particles.Linear Color=&FN_Select_Color_from_Time_Of_Day(Morning=<rgba 1 1 1 1>, Day=<rgba 1 1 1 1>, Evening=<rgba 1 1 1 1>, Night=<rgba 0.75 0.75 0.75 1>)) {isSetParameters=true}
   &AddVelocity(
    Velocity=<0 0 -150>
    Velocity Speed Scale=&FloatFromCurve(
     FloatCurve=DataInterface() {propertyValues {Curve {
        defaultValue=1
        keys [{interpMode=RCIM_Cubic, tangentMode=RCTM_User, time=0.5616, value=0}, {interpMode=RCIM_Cubic, tangentMode=RCTM_User, time=0.65, value=1}]
       }}}
     CurveIndex=$System.Age
     Scale Curve=&MaskFloat(Mask Source=0, Mask Parameter=?, Mask=0, Masked Float=1, Passthrough Float=0)
    )
   ) {module=AddVelocity001}
  ]
  particleUpdateScript [
   &ParticleState()
   &ScaleColor(
    Scale Alpha=&MaskFloat(
     Mask Source=0
     Mask Parameter=?
     Mask=0
     Masked Float=1
     Passthrough Float=&FloatFromCurve(
      FloatCurve=DataInterface() {propertyValues {Curve {
         defaultValue=1
         keys [
          {interpMode=RCIM_Cubic, tangentMode=RCTM_User, time=0.7778, value=1}
          {arriveTangent=-3.0163, interpMode=RCIM_Cubic, leaveTangent=-3.0163, tangentMode=RCTM_User, time=1.1605, value=0}
         ]
        }}}
      CurveIndex=$Emitter.Age
     )
    )
    ScaleA=true
    ScaleRGB=true
   )
   &ScaleRibbonWidth(Ribbon Width Scale=&MaskFloat(
     Mask Source=0
     Mask Parameter=?
     Mask=0
     Masked Float=&FloatFromCurve(FloatCurve=DataInterface() {propertyValues {Curve {keys [
          {arriveTangent=1.5802, interpMode=RCIM_Cubic, leaveTangent=1.5802, tangentMode=RCTM_User, time=0.0013, value=0}
          {arriveTangent=2.2322, interpMode=RCIM_Cubic, leaveTangent=2.2322, tangentMode=RCTM_User, time=0.1694, value=0.7635}
          {arriveTangent=0.0075, interpMode=RCIM_Cubic, leaveTangent=0.0075, tangentMode=RCTM_User, time=0.4, value=1.1}
          {arriveTangent=-5.9136, arriveTangentWeight=0.4225, interpMode=RCIM_Cubic, leaveTangent=-5.9136, tangentMode=RCTM_User, time=1, value=0}
         ]}}})
     Passthrough Float=&FloatFromCurve(FloatCurve=DataInterface() {propertyValues {Curve {keys [
          {arriveTangent=1.6896, interpMode=RCIM_Cubic, leaveTangent=1.6896, tangentMode=RCTM_User, time=0, value=0.9361}
          <at 0.9964 1.9884 RCIM_Cubic>
         ]}}})
    ))
   &Drag(Drag=5, Use Linear Drag=true, Use Rotational Drag=true)
   /Niagara/Modules/Update/Forces/V2/CurlNoiseForce(
    Noise Frequency=10
    Noise Strength=&MaskFloat(
     Mask Source=0
     Mask Parameter=?
     Mask=0
     Masked Float=5
     Passthrough Float=&FloatFromCurve(FloatCurve=DataInterface() {propertyValues {Curve {keys [<at 0 1>,<at 1 10>]}}}, CurveIndex=$Emitter.Age, Scale Curve=5)
    )
    Pan Noise=true
    Pan Noise Field=<0 0 0.1>
   )
   &SolveForcesAndVelocity(Clamp Velocity=false, Limit Acceleration=false)
  ]
  rendererClasses [&NiagaraRibbonRendererProperties]
  renderers [
   {rendererClass=&NiagaraRibbonRendererProperties}
   {rendererClass=&NiagaraRibbonRendererProperties, rendererIndex=1}
   {rendererClass=&NiagaraRibbonRendererProperties, rendererIndex=2}
  ]
 }
 FiresprintSubUV {
  emitterUpdateScript [
   &EmitterState(Life Cycle Mode=::NewEnumerator1, Loop Behavior=::NewEnumerator1, Loop Delay=0.1, Loop Duration=0.7, UseLoopDelay=true)
   &SpawnRate(SpawnGroup=3, SpawnRate=60, Use Spawn Probability=false) {module=SpawnRate003}
   &SpawnRate(
    SpawnGroup=2
    SpawnRate=&FloatFromCurve(
     FloatCurve=DataInterface() {propertyValues {Curve {keys [<at 0.4671 2 RCIM_Cubic>, {arriveTangent=-3.9119, interpMode=RCIM_Cubic, leaveTangent=-3.9119, tangentMode=RCTM_User, time=0.9775, value=1}]}}}
     CurveIndex=$Emitter.NormalizedLoopAge
     Scale Curve=160
    )
    Use Spawn Probability=false
   ) {module=SpawnRate002}
   &SpawnRate(SpawnRate=120, Use Spawn Probability=false) {module=SpawnRate001}
  ]
  particleSpawnScript [
   /SetVariables(Particles.Random size:NiagaraFloat=&RandomRangeFloat(Minimum=0.15, Maximum=0.25)) {isSetParameters=true}
   &InitializeParticle(
    Color=&MergeHelper_ColorSwitch(
     0=&RandomRangeLinearColor(Minimum=<rgba 1 0.2 0 0.5>, Maximum=<rgba 0.5 0.0417 0 0.3>)
     1=<rgba 1 1 1 1>
     2=&RandomRangeLinearColor(Minimum=<rgba 1 1 1 0.05>, Maximum=<rgba 1 1 1 0.225>)
     3=<rgba 1 1 1 1>
     Index=&SpawnGroup()
     Num=::NewEnumerator2
    )
    Lifetime=&MergeHelper_FloatSwitch(
     0=&RandomRangeFloat(Minimum=0.4, Maximum=0.7)
     1=&RandomRangeFloat(Minimum=0.3, Maximum=0.6)
     2=&RandomRangeFloat(Minimum=0.8, Maximum=0.6)
     3=&RandomRangeFloat(Minimum=0.4, Maximum=0.6)
     Index=&SpawnGroup()
     Num=::NewEnumerator2
    )
    Sprite Rotation Angle=&MergeHelper_FloatSwitch(
     0=&RandomRangeFloat(Minimum=0, Maximum=1)
     1=&SelectFloatFromArray(
      Array Sampling Mode=::NewEnumerator0
      Direct Array Index=?
      Float Selection Array=DataInterface() {propertyValues {FloatData [0.2,0.3,0.4,0.7,0.8,0.9,0.25,0.35,0.75,0.85,0.23,0.73]}}
      Random Seed=0
     )
     2=&RandomRangeFloat(Minimum=0, Maximum=1)
     3=?
     Index=&SpawnGroup()
     Num=::NewEnumerator1
    )
    Sprite Rotation Mode=::NewEnumerator1
    Sprite Size=&MergeHelper_Vector2Switch(
     0=<50 80>
     1=&Multiply_Vector2D_ByFloat(
      Float=&Multiply_Float(A=&RandomRangeFloat(Minimum=100, Maximum=120), B=$Particles.Random size)
      Vector2D=<1 1>
     )
     2=&Multiply_Vector2D_ByFloat(Float=&Multiply_Float(A=50, B=$Particles.Random size), Vector2D=<1 1>)
     3=?
     4=?
     Index=&SpawnGroup()
     Num=::NewEnumerator1
    )
    Sprite Size Mode=::NewEnumerator4
    UsePositionOffset=false
    Write Material Random=true
   )
   &SkeletalMeshLocation(
    Bone / Socket Index=&SpawnGroup()
    Bone Sampling Mode=::NewEnumerator4
    MaskSpawnGroup=false
    Orientation Sampling=::NewEnumerator1
    Sampled Position Offset=<0 20 0>
    Skeletal Mesh=$System.Skeletal Mesh Reader
   )
   /SetVariables(Particles.AORandFloat:NiagaraFloat=&UniformAOrBFloat(A=0, Distribution Weight=0.3, Evaluation Type=::NewEnumerator0, Recalculate Random Each Loop=true, Override Seed=false)) {isSetParameters=true}
   &ShapeLocation(
    MaskSpawnGroup=false
    Offset Coordinate Space=::Local
    Offset Mode=::NewEnumerator0
    Override Seed=false
    Shape Origin=$Particles.SkeletalMeshLocation.SampledPosition
    Sphere Radius=&MergeHelper_FloatSwitch(0=0, 1=0, 2=30, 3=0, Index=&SpawnGroup(), Num=::NewEnumerator2)
    UseOwnerScale=true
   )
   &AddVelocity(
    Velocity=&RotateVectorByQuaternion(Vector To Rotate=<-0.1 1 0.7>, Quaternion=$Particles.SkeletalMeshLocation.SampledRotation)
    Velocity Speed Scale=&Multiply_Float(
     A=200
     B=&MaskFloat(
      Mask Source=0
      Mask Parameter=?
      Mask=0
      Masked Float=&FloatFromCurve(
       FloatCurve=DataInterface() {propertyValues {Curve {
          defaultValue=1
          keys [
           {interpMode=RCIM_Cubic, tangentMode=RCTM_User, time=0.5278, value=1}
           {interpMode=RCIM_Cubic, tangentMode=RCTM_User, time=0.7975, value=0.4513}
          ]
         }}}
       CurveIndex=$Emitter.Age
      )
      Passthrough Float=1
     )
    )
   ) {module=AddVelocity001}
   /SetVariables(Particles.FireSpritesDynamic:Vector4f=<0 0 0 2.5>) {isSetParameters=true}
   /SetVariables(Particles.Initial Main flame Sprite Size:NiagaraFloat=&RandomRangeFloat(Minimum=35, Maximum=50)) {isSetParameters=true}
   /SetVariables(Particles.VisibilityTag=&SpawnGroup()) {isSetParameters=true}
   &AddVelocity(
    Velocity=<0 0 -150>
    Velocity Speed Scale=&FloatFromCurve(
     FloatCurve=DataInterface() {propertyValues {Curve {
        defaultValue=1
        keys [{interpMode=RCIM_Cubic, tangentMode=RCTM_User, time=0.5616, value=0}, {interpMode=RCIM_Cubic, tangentMode=RCTM_User, time=0.65, value=1}]
       }}}
     CurveIndex=$System.Age
    )
   )
   &CameraOffset(Camera Offset Amount=&MaskFloat(Mask Source=0, Mask Parameter=?, Mask=0, Masked Float=2, Passthrough Float=0))
  ]
  particleUpdateScript [
   &ParticleState()
   &ScaleSpriteSize(
    Scale Sprite Size Mode=::NewEnumerator0
    Uniform Scale Factor=&MergeHelper_FloatSwitch(
     0=&FloatFromCurve(FloatCurve=DataInterface() {propertyValues {Curve {
         defaultValue=0.2
         keys [
          {arriveTangent=0.7462, interpMode=RCIM_Cubic, leaveTangent=0.7462, tangentMode=RCTM_User, time=0, value=0.6}
          {arriveTangentWeight=0.3333, interpMode=RCIM_Cubic, tangentMode=RCTM_User, time=1, value=1}
         ]
        }}})
     1=&FloatFromCurve(
      FloatCurve=DataInterface() {propertyValues {Curve {keys [
          {arriveTangent=0.0679, interpMode=RCIM_Cubic, leaveTangent=0.0679, tangentMode=RCTM_User, time=0, value=0.2026}
          {interpMode=RCIM_Cubic, tangentMode=RCTM_User, time=0.377, value=0.3}
          {arriveTangent=-0.321, interpMode=RCIM_Cubic, leaveTangent=-0.321, time=0.9445, value=0.2535}
          {arriveTangent=-0.9658, tangentMode=RCTM_User, time=1, value=0.1}
         ]}}}
      Scale Curve=8
     )
     2=&FloatFromCurve(
      FloatCurve=DataInterface() {propertyValues {Curve {keys [{arriveTangent=1.0997, interpMode=RCIM_Cubic, leaveTangent=1.0997, tangentMode=RCTM_User, time=0, value=0.2}, <at 1 1.7902 RCIM_Cubic>]}}}
      Scale Curve=&RandomRangeFloat(Minimum=5, Maximum=3)
     )
     3=?
     Index=&SpawnGroup()
     Num=::NewEnumerator1
    )
   )
   &ScaleColor(
    Scale Alpha=&MergeHelper_FloatSwitch(
     0=&Multiply_Float(
      A=&FloatFromCurve(
       FloatCurve=DataInterface() {propertyValues {Curve {keys [
           {arriveTangent=9.0725, interpMode=RCIM_Cubic, leaveTangent=9.0725, tangentMode=RCTM_User, time=0, value=0}
           {arriveTangent=-2.5249, interpMode=RCIM_Cubic, leaveTangent=-2.5249, tangentMode=RCTM_User, time=0.1408, value=0.9847}
           {arriveTangent=-0.0405, interpMode=RCIM_Cubic, leaveTangent=-0.0405, tangentMode=RCTM_User, time=1, value=0}
          ]}}}
       Scale Curve=&FN_FrontEndFloatModifier(Gameplay=1, Front End=0.2)
      )
      B=&RandomRangeFloat(Minimum=1, Maximum=2)
     )
     1=&FloatFromCurve(
      FloatCurve=DataInterface() {propertyValues {Curve {keys [
          {arriveTangent=3.6955, interpMode=RCIM_Cubic, leaveTangent=3.6955, tangentMode=RCTM_User, time=0, value=0}
          {arriveTangent=0.6605, interpMode=RCIM_Cubic, leaveTangent=0.6605, time=0.3536, value=1}
          {arriveTangent=-1.547, interpMode=RCIM_Cubic, leaveTangent=-1.547, time=0.8004, value=0.5286}
          {tangentMode=RCTM_User, time=1, value=0}
         ]}}}
      Scale Curve=&Multiply_Float(
       A=&FN_FrontEndFloatModifier(Gameplay=0.8, Front End=0.6)
       B=&Lerp_Float(A=0.22, B=0.6, Alpha=$Particles.AORandFloat)
      )
     )
     2=&FloatFromCurve(FloatCurve=DataInterface() {propertyValues {Curve {
         defaultValue=0
         keys [
          {arriveTangent=4.0743, interpMode=RCIM_Cubic, leaveTangent=4.0743, tangentMode=RCTM_User, time=0, value=0}
          {arriveTangent=0.0034, interpMode=RCIM_Cubic, leaveTangent=0.0034, tangentMode=RCTM_User, time=0.3, value=0.9847}
          {arriveTangent=-3.0662, interpMode=RCIM_Cubic, leaveTangent=-3.0662, tangentMode=RCTM_User, time=1, value=0}
         ]
        }}})
     3=&Multiply_Float(
      A=&FloatFromCurve(FloatCurve=DataInterface() {propertyValues {Curve {keys [
           {arriveTangent=16.5612, interpMode=RCIM_Cubic, leaveTangentWeight=0.3333, time=0.1, value=0}
           {arriveTangentWeight=0.1333, interpMode=RCIM_Cubic, leaveTangentWeight=0.1667, tangentMode=RCTM_User, time=0.5, value=1}
           {arriveTangentWeight=0.1667, interpMode=RCIM_Cubic, tangentMode=RCTM_User, time=1, value=0}
          ]}}})
      B=&Multiply_Float(
       A=&RandomRangeFloat(Minimum=1, Maximum=1)
       B=&UniformAOrBFloat(A=0, Distribution Weight=1, Evaluation Type=::NewEnumerator0, Recalculate Random Each Loop=true, Override Seed=false)
      )
     )
     Index=&SpawnGroup()
     Num=::NewEnumerator2
    )
    Scale RGB=&MergeHelper_VectorSwitch(
     0=<1 1 1>
     1=&MakeVectorFromLinearColorRGB(Color=&MultiplyLinearColor(
       Linear Color=&LerpLinearColors(
        StartColor=&LerpLinearColors(
         StartColor=&ColorFromCurve(ColorCurve=DataInterface() {propertyValues {
            AlphaCurve {keys []}
            BlueCurve {keys [<at 0.2645 0.474>,<at 1 0.3855>]}
            GreenCurve {keys [<at 0.2645 3.2001>,<at 1 1.825>]}
            RedCurve {keys [<at 0.2645 13>,<at 1 7>]}
           }})
         EndColor=<rgba 0.0938 0.0138 0.003 1>
         LerpFactor=$Particles.AORandFloat
        )
        EndColor=&RandomRangeLinearColor(Minimum=<rgba 1 0.2 0 0.5>, Maximum=<rgba 0.5 0.0417 0 0.3>)
        LerpFactor=&FloatFromCurve(FloatCurve=DataInterface() {propertyValues {Curve {
            defaultValue=1
            keys [
             {arriveTangent=-1.6471, interpMode=RCIM_Cubic, leaveTangent=-1.6471, tangentMode=RCTM_User, time=0, value=1}
             {interpMode=RCIM_Cubic, tangentMode=RCTM_User, time=1, value=0}
            ]
           }}})
       )
       Scale Factor=&MakeVector4_FromLinearColor(LinearColor=&FNTimeOfDayColor(
         Night Color=<rgba 1 0.5409 0.1416 1>
         Morning Color=<rgba 1 0.5409 0.1416 1>
         Day Color=<rgba 1 0.5409 0.1416 1>
         Evening Color=<rgba 1 0.5409 0.1416 1>
        ))
      ))
     2=&MakeVectorFromLinearColorRGB(Color=&ColorFromCurve(ColorCurve=DataInterface() {propertyValues {
         AlphaCurve {keys [<at 0 1 RCIM_Cubic>,<at 1 0 RCIM_Cubic>]}
         BlueCurve {keys [<at 0 0 RCIM_Cubic>,<at 0.2816 0.0119>,<at 0.6981 0 RCIM_Cubic>]}
         GreenCurve {keys [<at 0 0.2694 RCIM_Cubic>,<at 0.2816 0.0059>,<at 0.6981 0 RCIM_Cubic>]}
         RedCurve {keys [<at 0 0.8189 RCIM_Cubic>,<at 0.2816 0.1753>,<at 0.6981 0 RCIM_Cubic>]}
        }}))
     3=&Multiply_VectorByFloat(
      Vector=&Multiply_VectorByFloat(
       Vector=<1 1 1>
       Float=&FloatFromCurve(FloatCurve=DataInterface() {propertyValues {Curve {keys [{arriveTangent=-5.044, interpMode=RCIM_Cubic, leaveTangent=-5.044, tangentMode=RCTM_User, time=0.25, value=2}, <at 0.4 1 RCIM_Cubic>]}}})
      )
      Float=&FloatFromCurve(
       FloatCurve=DataInterface() {propertyValues {Curve {keys [
           {arriveTangent=9.0725, interpMode=RCIM_Cubic, time=0, value=0}
           <at 0.5 0.9 RCIM_Cubic>
           {arriveTangent=-0.0405, interpMode=RCIM_Cubic, leaveTangent=-0.0405, tangentMode=RCTM_User, time=1, value=0}
          ]}}}
       Scale Curve=&Multiply_Float(A=&RandomRangeFloat(Minimum=0.5, Maximum=1), B=1)
      )
     )
     Index=&SpawnGroup()
     Num=::NewEnumerator2
    )
    ScaleA=true
    ScaleRGB=true
   )
   &SubUVAnimation(
    "Sprite Renderer "=DataInterface() {propertyValues {SpriteRenderer=&self:FiresprintSubUV_1.NiagaraSpriteRendererProperties_3}}
    UseEndFrame=true
    UseStartFrame=true
   )
   &Drag(
    Drag=&MaskFloat(
     Mask Source=0
     Mask Parameter=?
     Mask=0
     Masked Float=10
     Passthrough Float=&Multiply_Float(
      A=10
      B=&FloatFromCurve(FloatCurve=DataInterface() {propertyValues {Curve {
          defaultValue=1
          keys [{interpMode=RCIM_Cubic, tangentMode=RCTM_User, time=0, value=0}, {interpMode=RCIM_Cubic, tangentMode=RCTM_User, time=0.2385, value=1}]
         }}})
     )
    )
    Use Linear Drag=true
    Use Rotational Drag=true
   )
   &CurlNoiseForce(
    Bias Noise=false
    Noise Frequency=15
    Noise Strength=&MaskFloat(Mask Source=0, Mask Parameter=?, Mask=0, Masked Float=0, Passthrough Float=100)
    Pan Noise=true
    Pan Noise Field=<0 0 1>
   )
   &VortexForce(
    UseOriginOffset=false
    Vortex Force Amount=&MaskFloat(Mask Source=0, Mask Parameter=?, Mask=0, Masked Float=0, Passthrough Float=-2000)
   )
   &SolveForcesAndVelocity(Clamp Velocity=false, Limit Acceleration=false)
   /SetVariables(Particles.SmokeSubImageIndex:NiagaraFloat=$Particles.NormalizedAge) {isSetParameters=true}
   /SetVariables(
    "Particles.Main Fire DynParam ":Vector4f=&MakeVector4(
     X=&FloatFromCurve(
      FloatCurve=DataInterface() {propertyValues {Curve {
         defaultValue=1
         keys [{interpMode=RCIM_Cubic, tangentMode=RCTM_User, time=0, value=0}, {interpMode=RCIM_Cubic, tangentMode=RCTM_User, time=1.0015, value=0.6706}]
        }}}
      Scale Curve=0.85
     )
     Y=&RandomRangeFloat(Minimum=0, Maximum=4)
     Z=0
     W=2.5
    )
    Particles.Main Fire SpriteSize:Vector2f=&Multiply_Vector2D(
     A=&Vector2DFromFloat(Value=$Particles.Initial Main flame Sprite Size)
     B=&Multiply_Vector2D_ByFloat(
      Float=&FloatFromCurve(
       FloatCurve=DataInterface() {propertyValues {Curve {keys [
           {arriveTangent=1.4708, interpMode=RCIM_Cubic, leaveTangent=1.4708, tangentMode=RCTM_User, time=0, value=1.3}
           {arriveTangent=0.0624, interpMode=RCIM_Cubic, leaveTangent=0.0624, time=0.6, value=1.8}
           {interpMode=RCIM_Cubic, tangentMode=RCTM_User, time=1.0543, value=1.3657}
          ]}}}
       Scale Curve=&RandomRangeFloat(Minimum=2, Maximum=1)
      )
      Vector2D=<0.8 0.8>
     )
    )
    Particles.Main Flames color:LinearColor=&MakeLinearColorFromVectorAndFloat(
     "Vector (RGB)"=&VectorFromFloat(Value=&FloatFromCurve(
       FloatCurve=DataInterface() {propertyValues {Curve {keys [
           {arriveTangent=-0.4927, interpMode=RCIM_Cubic, leaveTangent=-0.4927, tangentMode=RCTM_User, time=0, value=0.1087}
           {arriveTangent=-0.5399, interpMode=RCIM_Cubic, leaveTangent=-0.5399, tangentMode=RCTM_User, time=0.1564, value=0.3915}
           {arriveTangent=-0.4927, interpMode=RCIM_Cubic, leaveTangent=-0.4927, tangentMode=RCTM_User, time=0.987, value=0.0549}
          ]}}}
       Scale Curve=&RandomRangeFloat(Minimum=2, Maximum=5)
      ))
     "Float (Alpha)"=&RandomRangeFloat(
      Minimum=&FloatFromCurve(
       FloatCurve=DataInterface() {propertyValues {Curve {keys [
           {arriveTangent=3.3783, interpMode=RCIM_Cubic, leaveTangent=3.3783, tangentMode=RCTM_User, time=0, value=0}
           {arriveTangent=0.2941, interpMode=RCIM_Cubic, leaveTangent=0.2941, tangentMode=RCTM_User, time=0.203, value=0.7175}
           {arriveTangent=-1.6053, interpMode=RCIM_Cubic, leaveTangent=-1.6053, time=0.5122, value=0.5742}
           {arriveTangent=-0.0646, interpMode=RCIM_Cubic, leaveTangent=-0.0646, tangentMode=RCTM_User, time=0.65, value=0}
          ]}}}
       Scale Curve=&Multiply_Float(A=&RandomRangeFloat(Minimum=0.3, Maximum=1), B=1)
      )
      Maximum=&FloatFromCurve(
       FloatCurve=DataInterface() {propertyValues {Curve {keys [
           {arriveTangent=3.3783, interpMode=RCIM_Cubic, leaveTangent=3.3783, tangentMode=RCTM_User, time=0, value=0}
           {arriveTangent=0.2941, interpMode=RCIM_Cubic, leaveTangent=0.2941, tangentMode=RCTM_User, time=0.203, value=0.7175}
           {arriveTangent=-1.109, interpMode=RCIM_Cubic, leaveTangent=-1.109, time=0.5122, value=0.5742}
           {arriveTangent=-0.0646, interpMode=RCIM_Cubic, leaveTangent=-0.0646, tangentMode=RCTM_User, time=0.85, value=0}
          ]}}}
       Scale Curve=&Multiply_Float(
        A=&RandomRangeFloat(Minimum=0.3, Maximum=1)
        B=&FNTimeOfDayScalar(Night Scale=0.5, Morning Scale=1, Evening Scale=1)
       )
      )
     )
    )
   ) {isSetParameters=true}
   /SetVariables() {isSetParameters=true}
   /SetVariables(Particles.FireSpriteSize:Vector2f=&Multiply_Vector2D(
     A=&Vector2DFromFloat(Value=&FloatFromCurve(
       FloatCurve=DataInterface() {propertyValues {Curve {keys [
           {arriveTangent=0.5131, interpMode=RCIM_Cubic, leaveTangent=0.1835, tangentMode=RCTM_Break, time=0, value=0.4}
           {interpMode=RCIM_Cubic, tangentMode=RCTM_Break, time=0.6, value=1}
           {arriveTangent=-3.5982, tangentMode=RCTM_User, time=1, value=0.5}
          ]}}}
       Scale Curve=&Lerp_Float(A=50, B=80, Alpha=&RandomRangeFloat(Minimum=0, Maximum=1))
      ))
     B=&Vector2DFromCurve(Vector2Curve=DataInterface() {propertyValues {XCurve {keys [<at 0 1>,<at 1 1>]}, YCurve {keys [<at 0 1>,<at 1 0.5>]}}}, ScaleCurve=<1 1>)
    )) {isSetParameters=true}
   &DynamicMaterialParameters(
    Index 0 Param 1=&RandomRangeFloat(Minimum=2, Maximum=1)
    Index 0 Param 2=&RandomRangeFloat(Minimum=0, Maximum=2)
    Index 0 Param 3=8
    Index 0 Param 4=$Particles.SubImageIndex
    Index 1 Param 1=&RandomRangeFloat(Minimum=-1, Maximum=1)
    Index 1 Param 2=&RandomRangeFloat(Minimum=-1, Maximum=1)
    Index 1 Param 3=&FloatFromCurve(FloatCurve=DataInterface() {propertyValues {Curve {keys [
         {arriveTangent=0.8752, interpMode=RCIM_Cubic, leaveTangent=0.8752, tangentMode=RCTM_User, time=0, value=0.5}
         {arriveTangent=0.9995, interpMode=RCIM_Cubic, leaveTangent=0.9995, tangentMode=RCTM_User, time=0.2223, value=0.6825}
         {arriveTangent=-0.0405, interpMode=RCIM_Cubic, leaveTangent=-0.0405, tangentMode=RCTM_User, time=0.7, value=0.9}
         <at 1 0.7>
        ]}}})
    Index 2 Param 4=1
    Module.Param0WriteEnabled=true
    Module.Param0WriteEnabled1=true
    Module.Param0WriteEnabled2=false
    Module.Param1WriteEnabled=true
    Module.Param1WriteEnabled1=true
    Module.Param1WriteEnabled2=false
    Module.Param2WriteEnabled=true
    Module.Param2WriteEnabled1=true
    Module.Param2WriteEnabled2=false
    Module.Param3WriteEnabled=true
    Module.Param3WriteEnabled1=false
    Module.Param3WriteEnabled2=true
    Write Parameter Index 1=true
    Write Parameter Index 2=true
   )
   /SetVariables(Particles.FlameSubImageIndex:NiagaraFloat=&FloatFromCurve(FloatCurve=DataInterface() {propertyValues {Curve {keys [<at 0 0>,<at 1 15>]}}})) {isSetParameters=true}
  ]
  rendererClasses [&NiagaraSpriteRendererProperties]
  renderers [
   {rendererClass=&NiagaraSpriteRendererProperties}
   {rendererClass=&NiagaraSpriteRendererProperties, rendererIndex=1}
   {rendererClass=&NiagaraSpriteRendererProperties, rendererIndex=2}
  ]
 }
 MouthGlow {
  emitterUpdateScript [
   &EmitterState(Life Cycle Mode=::NewEnumerator1, Loop Behavior=::NewEnumerator1, Loop Delay=0.2, Loop Duration=0.8, UseLoopDelay=true)
   &SpawnBurst_Instantaneous(Spawn Group=1, Use Spawn Probability=false, UseLoopCountLimit=false) {module=SpawnBurst_Instantaneous001}
  ]
  particleSpawnScript [
   &InitializeParticle(
    Color Mode=::NewEnumerator0
    Lifetime=0.7
    Sprite Size Mode=::NewEnumerator3
    Uniform Sprite Size=23
    UsePositionOffset=false
    Write Material Random=true
   )
   &CameraOffset(Camera Offset Amount=10)
   /SetVariables(Particles.VisibilityTag=&SpawnGroup()) {isSetParameters=true}
  ]
  particleUpdateScript [
   &ParticleState()
   &SkeletalMeshLocation(Bone Sampling Mode=::NewEnumerator4, MaskSpawnGroup=false, Sampled Position Offset=<0 10 -20>, Skeletal Mesh=$System.Skeletal Mesh Reader)
   &ScaleColor(
    Linear Color Curve=DataInterface() {propertyValues {
      AlphaCurve {keys [<at 0 0>,<at 0.5045 1>,<at 1 0>]}
      BlueCurve {keys [<at 0 0>,<at 1 0>]}
      GreenCurve {keys [<at 0 0.8608>,<at 1 0.2869>]}
      RedCurve {keys [<at 0 3>,<at 1 1>]}
     }}
    Scale Mode=2
   )
   &Light_Attributes(
    Color=1
    Color Value To Scale:LinearColor=&Multiply_LinearColorByFloat(Linear Color=<rgba 2 0.5217 0 5>, Float=1)
    Falloff Exponent=1
    Radius=1
    Scale Alpha:NiagaraFloat=&FloatFromCurve(FloatCurve=DataInterface() {propertyValues {Curve {
        defaultValue=1
        keys [
         {interpMode=RCIM_Cubic, tangentMode=RCTM_User, time=0, value=0}
         {arriveTangent=2, interpMode=RCIM_Cubic, leaveTangent=2, time=0.0799, value=0.8593}
         {interpMode=RCIM_Cubic, tangentMode=RCTM_User, time=0.5, value=1}
         {tangentMode=RCTM_User, time=1, value=0}
        ]
       }}})
    ScaleA=true
    ScaleRGB=true
   )
  ]
  rendererClasses [&NiagaraLightRendererProperties]
  renderers [{rendererClass=&NiagaraLightRendererProperties, rendererIndex=0}]
 }
 SplineFire {
  emitterUpdateScript [
   &EmitterState(Life Cycle Mode=::NewEnumerator1, Loop Behavior=::NewEnumerator1, Loop Delay=0.33, Loop Duration=1, UseLoopDelay=true)
   &SpawnBurst_Instantaneous(Use Spawn Probability=false, UseLoopCountLimit=false)
   /SetVariables(Emitter.Dissolve:NiagaraFloat=&FloatFromCurve(
     FloatCurve=DataInterface() {propertyValues {Curve {
        defaultValue=1
        keys [
         {interpMode=RCIM_Cubic, tangentMode=RCTM_User, time=0.7423, value=0}
         {arriveTangent=-0.2679, interpMode=RCIM_Cubic, leaveTangent=-0.2679, tangentMode=RCTM_User, time=1.2, value=1}
        ]
       }}}
     CurveIndex=$Emitter.Age
     Scale Curve=5
    )) {isSetParameters=true}
  ]
  particleSpawnScript [
   &InitializeParticle(
    Color=<rgba 3 0.6053 0 3>
    Lifetime=&Lerp_Float(
     A=&RandomRangeFloat(Minimum=0.65, Maximum=0.3)
     B=&RandomRangeFloat(Minimum=0.65, Maximum=0.85)
     Alpha=&FloatFromCurve(
      FloatCurve=DataInterface() {propertyValues {Curve {
         defaultValue=1
         keys [
          {arriveTangent=1.6624, interpMode=RCIM_Cubic, leaveTangent=1.6624, tangentMode=RCTM_User, time=0, value=0}
          {interpMode=RCIM_Cubic, tangentMode=RCTM_User, time=1, value=1}
         ]
        }}}
      CurveIndex=$Emitter.Age
     )
    )
    Mesh Scale Mode=::NewEnumerator1
    Mesh Uniform Scale Max=0.7
    Mesh Uniform Scale Min=0.5
    UsePositionOffset=false
    Write Material Random=true
   )
   /SetVariables(Particles.ScaleOverTime:NiagaraFloat=&FloatFromCurve(
     FloatCurve=DataInterface() {propertyValues {Curve {
        defaultValue=1
        keys [
         {interpMode=RCIM_Cubic, tangentMode=RCTM_User, time=0, value=0}
         {arriveTangent=1.5988, interpMode=RCIM_Cubic, leaveTangent=1.5988, tangentMode=RCTM_User, time=0.811, value=1}
        ]
       }}}
     CurveIndex=$Emitter.Age
    )) {isSetParameters=true}
   &InitialMeshOrientation(
    Mesh Axis To Orient=<0 1 0>
    Mesh Orientation Mode=::NewEnumerator2
    Orientation Vector=<0 0 1>
    Rotation=<0 0.5 0.0413>
    Use Rotation Vector=true
   )
   /SetVariables(
    Particles.Position=&AddVectorToPosition(
     Position=$Engine.Emitter.SimulationPosition
     Vector=&Lerp_Vector(A=<0 0 10>, B=&RandomRangeVector(Minimum=<0 0 60>, Maximum=<0 0 80>), Alpha=$Particles.ScaleOverTime)
    )
    Particles.Scale=&VectorFromFloat(Value=&Lerp_Float(A=0.5, B=0.5, Alpha=$Particles.ScaleOverTime))
   ) {isSetParameters=true}
  ]
  particleUpdateScript [
   &ParticleState()
   &ScaleMeshSize(Scale Factor=&MergeHelper_VectorSwitch(
     0=<1 1 1>
     1=&MakeVector(
      X=1
      Y=1
      Z=&FloatFromCurve(FloatCurve=DataInterface() {propertyValues {Curve {
          defaultValue=1
          keys [{interpMode=RCIM_Cubic, tangentMode=RCTM_User, time=0, value=1}, {interpMode=RCIM_Cubic, tangentMode=RCTM_User, time=1, value=2}]
         }}})
     )
     2=?
     3=?
     Index=&SpawnGroup()
     Num=::NewEnumerator0
    ))
   &UpdateMeshOrientation(
    Rotation Rate=&FloatFromCurve(
     FloatCurve=DataInterface() {propertyValues {Curve {
        defaultValue=1
        keys [{interpMode=RCIM_Cubic, tangentMode=RCTM_User, time=0, value=1}, {interpMode=RCIM_Cubic, tangentMode=RCTM_User, time=1, value=0.2729}]
       }}}
     Scale Curve=&RandomRangeFloat(Minimum=-1, Maximum=-2)
    )
    Rotation Vector=<0 1 0>
   )
   &DynamicMaterialParameters(
    Index 0 Param 1=&FloatFromCurve(
     FloatCurve=DataInterface() {propertyValues {Curve {
        defaultValue=1
        keys [
         {tangentMode=RCTM_User, time=0, value=1}
         {arriveTangent=-0.0405, interpMode=RCIM_Cubic, leaveTangent=-0.0405, tangentMode=RCTM_User, time=0.8, value=1}
         {arriveTangent=4.0917, interpMode=RCIM_Cubic, leaveTangent=4.0917, tangentMode=RCTM_User, time=1, value=1.2}
        ]
       }}}
     Scale Curve=0.6
    )
    Index 0 Param 2=&MaskFloat(
     Mask Source=0
     Mask Parameter=?
     Mask=0
     Masked Float=0
     Passthrough Float=&FloatFromCurve(
      FloatCurve=DataInterface() {propertyValues {Curve {defaultValue=0, keys [{tangentMode=RCTM_User, time=0, value=1}, <at 0.2378 0.5162>, {tangentMode=RCTM_User, time=1, value=0.3609}]}}}
      Scale Curve=15
     )
    )
    Index 0 Param 3=&FloatFromCurve(
     FloatCurve=DataInterface() {propertyValues {Curve {
        defaultValue=1
        keys [
         {tangentMode=RCTM_User, time=0, value=0}
         {arriveTangent=-0.0405, interpMode=RCIM_Cubic, leaveTangent=-0.0405, tangentMode=RCTM_User, time=0.6, value=0}
         {arriveTangent=16.8033, interpMode=RCIM_Cubic, leaveTangent=16.8033, tangentMode=RCTM_User, time=1, value=1}
        ]
       }}}
     Scale Curve=7
    )
    Index 0 Param 4=&Add_Float(
     A=&FloatFromCurve(
      FloatCurve=DataInterface() {propertyValues {Curve {
         defaultValue=1
         keys [
          {arriveTangent=9.0725, interpMode=RCIM_Cubic, leaveTangent=9.0725, tangentMode=RCTM_User, time=0, value=0}
          {arriveTangent=-2.5249, interpMode=RCIM_Cubic, leaveTangent=-2.5249, tangentMode=RCTM_User, time=0.06, value=0.9847}
          {arriveTangent=-0.0405, interpMode=RCIM_Cubic, leaveTangent=-0.0405, tangentMode=RCTM_User, time=1, value=0}
         ]
        }}}
      Scale Curve=-0.5
     )
     B=&RandomRangeFloat(Minimum=0, Maximum=2)
    )
    Module.Index0Float0Write=true
    Module.Index0Float1Write=true
    Module.Index0Float2Write=true
    Module.Index0Float3Write=true
   )
   &ScaleColor(
    Scale Alpha=&FloatFromCurve(
     FloatCurve=DataInterface() {propertyValues {Curve {
        defaultValue=0
        keys [
         <at 0.1945 0 RCIM_Cubic>
         {arriveTangent=0.2081, interpMode=RCIM_Cubic, leaveTangent=0.2081, tangentMode=RCTM_User, time=0.3914, value=1}
         {arriveTangent=-0.2972, interpMode=RCIM_Cubic, leaveTangent=-0.2972, tangentMode=RCTM_User, time=1, value=0}
        ]
       }}}
     Scale Curve=&Multiply_Float(
      A=&FNTimeOfDayScalar(Night Scale=0.4, Morning Scale=0.8, Evening Scale=0.8)
      B=&FN_FrontEndFloatModifier(Gameplay=0.5, Front End=0.3)
     )
    )
    ScaleA=true
    ScaleRGB=true
   )
   /SetVariables(Particles.SoftColor:LinearColor=&MultiplyLinearColor(
     Linear Color=<rgba 0.6 0.2 0 1>
     Scale Factor=&MakeVector4FromVectorAndFloat(
      "Vector (RGB)"=<1 1 1>
      "Float (Alpha)"=&FloatFromCurve(
       FloatCurve=DataInterface() {propertyValues {Curve {
          defaultValue=0
          keys [
           <at 0.197 0 RCIM_Cubic>
           {arriveTangent=0.2081, interpMode=RCIM_Cubic, leaveTangent=0.2081, tangentMode=RCTM_User, time=0.4637, value=1}
           {arriveTangent=-0.2972, interpMode=RCIM_Cubic, leaveTangent=-0.2972, tangentMode=RCTM_User, time=1, value=0}
          ]
         }}}
       Scale Curve=&Multiply_Float(
        A=&FNTimeOfDayScalar(Night Scale=0.6, Morning Scale=0.8, Evening Scale=0.8)
        B=&FN_FrontEndFloatModifier(Gameplay=3, Front End=1.5)
       )
      )
     )
    )) {isSetParameters=true}
   /SetVariables(Particles.DynSoft:Vector4f=&MakeVector4(
     X=&FloatFromCurve(FloatCurve=DataInterface() {propertyValues {Curve {
         defaultValue=1
         keys [
          {tangentMode=RCTM_User, time=0, value=1}
          {arriveTangent=-0.0405, interpMode=RCIM_Cubic, leaveTangent=-0.0405, tangentMode=RCTM_User, time=0.6, value=1}
          {arriveTangent=4.0917, interpMode=RCIM_Cubic, leaveTangent=4.0917, tangentMode=RCTM_User, time=1, value=2}
         ]
        }}})
     Y=0
     Z=&FloatFromCurve(
      FloatCurve=DataInterface() {propertyValues {Curve {
         defaultValue=1
         keys [
          {tangentMode=RCTM_User, time=0, value=0}
          {arriveTangent=-0.0405, interpMode=RCIM_Cubic, leaveTangent=-0.0405, tangentMode=RCTM_User, time=0.6, value=0}
          {arriveTangent=5.6743, interpMode=RCIM_Cubic, leaveTangent=5.6743, tangentMode=RCTM_User, time=1, value=1}
         ]
        }}}
      Scale Curve=20
     )
     W=&Add_Float(
      A=&FloatFromCurve(
       FloatCurve=DataInterface() {propertyValues {Curve {
          defaultValue=1
          keys [
           {arriveTangent=9.0725, interpMode=RCIM_Cubic, leaveTangent=9.0725, tangentMode=RCTM_User, time=0, value=0}
           {arriveTangent=-2.5249, interpMode=RCIM_Cubic, leaveTangent=-2.5249, tangentMode=RCTM_User, time=0.06, value=0.9847}
           {arriveTangent=-0.0405, interpMode=RCIM_Cubic, leaveTangent=-0.0405, tangentMode=RCTM_User, time=1, value=0}
          ]
         }}}
       Scale Curve=-0.8
      )
      B=&RandomRangeFloat(Minimum=0, Maximum=1)
     )
    )) {isSetParameters=true}
   /SetVariables(Particles.DynSoft2:Vector4f=&MakeVector4(
     X=&FloatFromCurve(
      FloatCurve=DataInterface() {propertyValues {Curve {
         defaultValue=1
         keys [
          {tangentMode=RCTM_User, time=0, value=1}
          {arriveTangent=-0.0405, interpMode=RCIM_Cubic, leaveTangent=-0.0405, tangentMode=RCTM_User, time=0.6, value=1}
          {arriveTangent=4.0917, interpMode=RCIM_Cubic, leaveTangent=4.0917, tangentMode=RCTM_User, time=1, value=2}
         ]
        }}}
      Scale Curve=0.9
     )
     Y=&MaskFloat(
      Mask Source=0
      Mask Parameter=?
      Mask=0
      Masked Float=0
      Passthrough Float=&FloatFromCurve(
       FloatCurve=DataInterface() {propertyValues {Curve {defaultValue=0, keys [{tangentMode=RCTM_User, time=0, value=1}, <at 0.2378 0.5162>, {tangentMode=RCTM_User, time=1, value=0.3609}]}}}
       Scale Curve=15
      )
     )
     Z=&FloatFromCurve(
      FloatCurve=DataInterface() {propertyValues {Curve {
         defaultValue=1
         keys [
          {tangentMode=RCTM_User, time=0, value=0}
          {arriveTangent=-0.0405, interpMode=RCIM_Cubic, leaveTangent=-0.0405, tangentMode=RCTM_User, time=0.6, value=0}
          {arriveTangent=10.8261, interpMode=RCIM_Cubic, leaveTangent=10.8261, tangentMode=RCTM_User, time=1, value=1}
         ]
        }}}
      Scale Curve=12
     )
     W=&Add_Float(
      A=&FloatFromCurve(
       FloatCurve=DataInterface() {propertyValues {Curve {
          defaultValue=1
          keys [
           {arriveTangent=9.0725, interpMode=RCIM_Cubic, leaveTangent=9.0725, tangentMode=RCTM_User, time=0, value=0}
           {arriveTangent=-2.5249, interpMode=RCIM_Cubic, leaveTangent=-2.5249, tangentMode=RCTM_User, time=0.06, value=0.9847}
           {arriveTangent=-0.0405, interpMode=RCIM_Cubic, leaveTangent=-0.0405, tangentMode=RCTM_User, time=1, value=0}
          ]
         }}}
       Scale Curve=-0.65
      )
      B=&RandomRangeFloat(Minimum=0, Maximum=1)
     )
    )) {isSetParameters=true}
  ]
  rendererClasses [&NiagaraMeshRendererProperties]
  renderers [
   {rendererClass=&NiagaraMeshRendererProperties, rendererIndex=0}
   {rendererClass=&NiagaraMeshRendererProperties, rendererIndex=1}
   {rendererClass=&NiagaraMeshRendererProperties, rendererIndex=2}
  ]
 }
}
summary {
 systemName=NS_Emote_SandyInland_Fire
 emitters [
  {emitterName=FiresprintSubUV, rendererClasses [&NiagaraSpriteRendererProperties]}
  {emitterName=Embers001, rendererClasses [&NiagaraSpriteRendererProperties]}
  {emitterName=Fire, rendererClasses [&NiagaraRibbonRendererProperties]}
  {emitterName=SplineFire, rendererClasses [&NiagaraMeshRendererProperties]}
  {emitterName=MouthGlow, rendererClasses [&NiagaraLightRendererProperties]}
 ]
 userVariables [
  User.bIsFrontEnd {type=NiagaraBool}
  User.bIsFrontEndPreview {type=NiagaraBool}
  User.LifeMultiplier {type=NiagaraFloat, defaultValue=0.56, description=""}
  User.Reveal {type=NiagaraFloat, defaultValue=0.5, description=""}
  User.ScaleLifetime {type=NiagaraFloat, defaultValue=2.5, description=""}
  User.Variant {type=NiagaraInt32}
  User.VelocityMultiplier {type=NiagaraFloat, defaultValue=1.2, description=""}
 ]
}
